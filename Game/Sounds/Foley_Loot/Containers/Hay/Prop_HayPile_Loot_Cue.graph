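# Sound cue for searching a hay pile loot prop, in the loot collect sound class with prop search attenuation.
# A random node picks one of two hay search waves at equal weight, then feeds a modulator at default settings.
# An enveloper at the root fades volume from full to silent over two seconds.
# A concurrency override prevents new instances from starting once the limit is reached.
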SoundCue /Game/Sounds/Foley_Loot/Containers/Hay/Prop_HayPile_Loot_Cue
root=Enveloper_2
nodes [
 {
  id=Enveloper_2
  children [Modulator_2]
  properties {pitchCurve {editorCurveData {keys [{time=0}]}}, volumeCurve {editorCurveData {keys [<at 0 1>,<at 2 0>]}}}
 }
 {id=Modulator_2, children [Random_7]}
 {id=Random_7, children [WavePlayer_18, WavePlayer_19], properties {weights [1,1]}}
 {id=WavePlayer_18, properties {soundWaveAssetPtr=&/hay_search_01}}
 {id=WavePlayer_19, properties {soundWaveAssetPtr=&/hay_search_02}}
]
properties {
 attenuationSettings=/Game/Sounds/Attenuation/Fort_Prop_Search_Att
 bOverrideConcurrency=true
 maxDistance=1300
 soundClassObject=&Fort_Loot_Collect
 volumeMultiplier=1
 concurrencyOverrides {resolutionRule=PreventNew}
}
