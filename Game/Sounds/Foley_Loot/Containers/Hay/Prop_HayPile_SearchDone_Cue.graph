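# Sound cue for finishing a hay pile search, mixing a "no more loot" UI sound with a hay rustle.
# The rustle is picked at random from two equally weighted waves, passed through a modulator, then faded out by an enveloper.
# Playback uses prop-search attenuation and the loot-collect sound class, with a concurrency override allowing one instance and blocking new ones.
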SoundCue /Game/Sounds/Foley_Loot/Containers/Hay/Prop_HayPile_SearchDone_Cue
root=Mixer_1
nodes [
 {
  id=Enveloper_1
  children [Modulator_3]
  properties {pitchCurve {editorCurveData {keys [{time=0}]}}, volumeCurve {editorCurveData {keys [<at 0 1>,<at 2 0>]}}}
 }
 {id=Mixer_1, children [WavePlayer_8,Enveloper_1], properties {inputVolume [0.8,0.8]}}
 {id=Modulator_3, children [Random_8]}
 {id=Random_8, children [WavePlayer_20, WavePlayer_21], properties {weights [1,1]}}
 {id=WavePlayer_20, properties {soundWaveAssetPtr=&/hay_search_01}}
 {id=WavePlayer_21, properties {soundWaveAssetPtr=&/hay_search_02}}
 {id=WavePlayer_8, properties {soundWaveAssetPtr=/Game/Sounds/Fort_UI_Sounds/Gameplay/Fort_UI_NoMoreLoot}}
]
properties {
 attenuationSettings=/Game/Sounds/Attenuation/Fort_Prop_Search_Att
 bOverrideConcurrency=true
 maxDistance=1300
 soundClassObject=&Fort_Loot_Collect
 volumeMultiplier=0.7
 concurrencyOverrides {maxCount=1, resolutionRule=PreventNew}
}
