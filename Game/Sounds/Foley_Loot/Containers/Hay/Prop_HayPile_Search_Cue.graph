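# Sound cue for searching a hay pile prop, playing as loot container foley.
# A Mixer layers one of two random hay search waves, passed through a Modulator at reduced volume, with one of three random generic gear search waves at boosted volume.
# It uses the prop search attenuation and loot collect sound class, and its concurrency override prevents new instances at the limit.
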SoundCue /Game/Sounds/Foley_Loot/Containers/Hay/Prop_HayPile_Search_Cue
root=Mixer_26
nodes [
 {id=Mixer_26, children [Modulator_2,Random_25], properties {inputVolume [0.4,1.2]}}
 {id=Modulator_2, children [Random_7]}
 {id=Random_25, children [WavePlayer_129, WavePlayer_130, WavePlayer_132], properties {weights [1,1,1]}}
 {id=Random_7, children [WavePlayer_18, WavePlayer_19], properties {weights [1,1]}}
 {id=WavePlayer_129, properties {soundWaveAssetPtr=/Game/Sounds/Foley_Loot/Containers/Misc/Prop_Gear_Search_05}}
 {id=WavePlayer_130, properties {soundWaveAssetPtr=/Game/Sounds/Foley_Loot/Containers/Misc/Prop_Gear_Search_06}}
 {id=WavePlayer_132, properties {soundWaveAssetPtr=/Game/Sounds/Foley_Loot/Containers/Misc/Prop_Gear_Search_01}}
 {id=WavePlayer_18, properties {soundWaveAssetPtr=&/hay_search_01}}
 {id=WavePlayer_19, properties {soundWaveAssetPtr=&/hay_search_02}}
]
properties {
 attenuationSettings=/Game/Sounds/Attenuation/Fort_Prop_Search_Att
 bOverrideConcurrency=true
 maxDistance=1300
 soundClassObject=&Fort_Loot_Collect
 volumeMultiplier=1.3
 concurrencyOverrides {resolutionRule=PreventNew}
}
